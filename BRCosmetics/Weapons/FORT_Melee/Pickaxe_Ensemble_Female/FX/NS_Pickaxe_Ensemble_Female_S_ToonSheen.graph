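# Niagara cosmetic effect for the Ensemble Female pickaxe, playing an intermittent toon sheen and sparkle on the axe head.
# AxeHead_Sheen bursts one short-lived mesh particle per randomized emitter loop, animating a dynamic material parameter over its life via curve.
# That particle generates a death event, though no receiving handler is shown.
# Axehead_Sheen_Sparkle renders an offset, camera-pulled sprite that shrinks and rotates; it shows no spawn module of its own.
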NiagaraSystem /BRCosmetics/Weapons/FORT_Melee/Pickaxe_Ensemble_Female/FX/NS_Pickaxe_Ensemble_Female_S_ToonSheen
dependencies {
 usedDataInterfaces [&NiagaraDataInterfaceCurve, &NiagaraDataInterfaceMeshRendererInfo]
 usedDynamicInputs []
 usedModules [&SystemState]
}
emitters {
 AxeHead_Sheen {
  emitterUpdateScript [
   &EmitterState(
    Life Cycle Mode=::NewEnumerator1
    Loop Delay=&UniformRangedFloat(Minimum=0.5, Maximum=1)
    Loop Duration=&UniformRangedFloat(Minimum=4, Maximum=9)
    Recalculate Duration Each Loop=true
    UseLoopDelay=true
   )
   &SpawnBurst_Instantaneous(Spawn Count=1, Use Spawn Probability=false)
   &SpawnRate(SpawnRate=5, Use Spawn Probability=false) {enabled=false}
  ]
  particleSpawnScript [
   &InitializeParticle(Lifetime=0.333, UsePositionOffset=false, Write Material Random=true)
   &CameraOffset(Camera Offset Amount=0.05)
  ]
  particleUpdateScript [
   &ScaleColor(
    Scale Alpha=&FloatFromCurve(FloatCurve=DataInterface() {propertyValues {Curve {keys [<at 0.5016 0.9767>]}}})
    Scale RGB=&VectorFromFloat(Value=&FloatFromCurve(FloatCurve=DataInterface() {propertyValues {Curve {keys [<at 0.5 0.8>]}}}))
    ScaleA=true
    ScaleRGB=true
   )
   &ParticleState()
   &DynamicMaterialParameters(
    Index 0 Param 1=&FloatFromCurve(FloatCurve=DataInterface() {propertyValues {Curve {keys [
         {arriveTangent=0.2595, interpMode=RCIM_Cubic, leaveTangent=0.2595, tangentMode=RCTM_User, time=0, value=0}
         {arriveTangent=-0.0277, interpMode=RCIM_Cubic, leaveTangent=-0.0277, tangentMode=RCTM_User, time=1.1, value=1.8}
        ]}}})
    Index 0 Param 2=1
    Index 0 Param 3=1
    Index 0 Param 4=1
    Module.Param0WriteEnabled=true
    Module.Param1WriteEnabled=true
    Module.Param2WriteEnabled=true
    Module.Param3WriteEnabled=true
   )
   &SolveForcesAndVelocity(Clamp Velocity=false, Limit Acceleration=false)
   &GenerateDeathEvent()
  ]
  rendererClasses [&NiagaraMeshRendererProperties]
  renderers [{rendererClass=&NiagaraMeshRendererProperties, rendererIndex=0}]
 }
 Axehead_Sheen_Sparkle {
  emitterUpdateScript [&EmitterState(Life Cycle Mode=::NewEnumerator1, Loop Duration=1, Recalculate Duration Each Loop=true, UseLoopDelay=true)]
  particleSpawnScript [
   &InitializeParticle(
    Lifetime Max=0.7
    Lifetime Min=0.5
    Lifetime Mode=::NewEnumerator1
    Position Offset=<0.6 -0.25 35.75>
    Sprite Rotation Mode=::NewEnumerator2
    Sprite Size Mode=::NewEnumerator1
    Uniform Sprite Size Max=16
    Uniform Sprite Size Min=12
    UsePositionOffset=true
    Write Material Random=true
   ) {module=InitializeParticle001}
   &CameraOffset(Camera Offset Amount=15)
  ]
  particleUpdateScript [
   &ParticleState()
   &ScaleSpriteSize(Scale Factor=&Vector2DFromFloat(Value=&FloatFromCurve(
      FloatCurve=DataInterface() {propertyValues {Curve {keys [{arriveTangent=-1.2765, interpMode=RCIM_Cubic, leaveTangent=-1.2765, tangentMode=RCTM_User, time=0, value=1}, <at 1 0>]}}}
      Scale Curve=0.65
     )))
   /SetVariables(Particles.SpriteRotation:NiagaraFloat=&Add_Float(
     A=$Particles.Initial.SpriteRotation
     B=&FloatFromCurve(
      FloatCurve=DataInterface() {propertyValues {Curve {keys [{arriveTangent=-2.8584, interpMode=RCIM_Cubic, leaveTangent=-2.8584, tangentMode=RCTM_User, time=0, value=1}, <at 1 0>]}}}
      Scale Curve=180
     )
    )) {isSetParameters=true}
  ]
  rendererClasses [&NiagaraSpriteRendererProperties]
  renderers [{rendererClass=&NiagaraSpriteRendererProperties}]
 }
}
summary {
 systemName=NS_Pickaxe_Ensemble_Female_S_ToonSheen
 emitters [
  {emitterName=AxeHead_Sheen, rendererClasses [&NiagaraMeshRendererProperties]}
  {emitterName=Axehead_Sheen_Sparkle, rendererClasses [&NiagaraSpriteRendererProperties]}
 ]
 userVariables []
}
